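# Physics asset of collision shapes for the Ice Queen medium female body, named as colliders for the cape.
# Capsules cover the head, clavicles, pelvis, spine and feet, with spine_02 to spine_04 each pairing two at different orientations.
# Further capsules sit on back, mid, mid-back and side dyn_skirt bones, so colliders follow the skirt.
# Arm, thigh, calf and neck_02 bodies hold no shapes, and no constraints link any bodies.
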PhysicsAsset /Game/Characters/Player/Female/Medium/Bodies/F_MED_IceQueen/Meshes/F_MED_Ice_Queen_CapeColliders
bodies {
 calf_l []
 calf_r []
 clavicle_l [{shapeType=Capsule, length=7.6964, radius=9.3853, center=<11.6458 -0.2785 1.8018>, rotation=<rot -77.5566 177.3915 7.7314>}]
 clavicle_r [{shapeType=Capsule, length=7.6964, radius=9.7957, center=<-11.0409 0.251 0.9424>, rotation=<rot -77.5565 177.3915 -172.2685>}]
 dyn_skirt_back_07 [{
   shapeType=Capsule
   length=18
   radius=12
   shapeName=dyn_skirt_back_07_capsule
   center=<-0.3629 5.4142 0>
   rotation=<rot -90 179.9954 -179.9946>
  }]
 dyn_skirt_mid_05_l [{
   shapeType=Capsule
   length=18.75
   radius=15
   shapeName=dyn_skirt_mid_05_l_capsule
   center=<2.861 5.4812 0>
   rotation=<rot -90 179.9965 -179.9961>
  }]
 dyn_skirt_mid_05_r [{shapeType=Capsule, length=24.5, radius=15, shapeName=dyn_skirt_mid_05_r_capsule, center=<-1.9961 3.4876 0>, rotation=<rot -90 0 0.0008>}]
 dyn_skirt_mid_back_05_l [{
   shapeType=Capsule
   length=27.25
   radius=15
   shapeName=dyn_skirt_mid_back_05_l_capsule
   center=<-14.9635 7.4229 0>
   rotation=<rot -90 179.9902 -179.9895>
  }]
 dyn_skirt_mid_back_05_r [{
   shapeType=Capsule
   length=27.25
   radius=15
   shapeName=dyn_skirt_mid_back_05_r_capsule
   center=<-14.9635 5.1056 0>
   rotation=<rot -90 -179.9406 179.9414>
  }]
 dyn_skirt_side_05_l [{
   shapeType=Capsule
   length=28.25
   radius=15
   shapeName=dyn_skirt_side_05_l_capsule
   center=<-2.1521 8.9188 0>
   rotation=<rot -90 169.975 -169.9752>
  }]
 dyn_skirt_side_05_r [{
   shapeType=Capsule
   length=28.25
   radius=15
   shapeName=dyn_skirt_side_05_r_capsule
   center=<-1.3965 6.9252 0>
   rotation=<rot -90 -179.9802 179.9805>
  }]
 foot_l [{shapeType=Capsule, length=12.5982, radius=6.5127, center=<-2.2072 6.0534 -0.3367>, rotation=<rot 0.0388 1.4701 -89.3982>}]
 foot_r [{shapeType=Capsule, length=12.5982, radius=6.5127, center=<2.2072 -6.0534 0.3367>, rotation=<rot 0.0388 1.4701 90.6018>}]
 head [{shapeType=Capsule, length=0, radius=18, center=<9.0697 0.0746 0>, rotation=<rot 0 -90 0>}]
 lowerarm_l []
 lowerarm_r []
 neck_02 []
 pelvis [{shapeType=Capsule, length=9.4028, radius=20.4095, center=<-7.3414 -0.7485 -0.1981>}]
 spine_01 [{shapeType=Capsule, length=9.7671, radius=12.4854, center=<-1.3851 -1.0873 0>}]
 spine_02 [
  {shapeType=Capsule, length=3.2984, radius=8.4854, center=<1.179 2.2974 0>}
  {shapeType=Capsule, length=20.6875, radius=9.3125, center=<-1.8317 2.606 0>, rotation=<rot 90 0 0>}
 ]
 spine_03 [
  {shapeType=Capsule, length=4.2984, radius=11.2354, center=<-0.0747 2.3725 0>}
  {shapeType=Capsule, length=20.7188, radius=9.2812, center=<-1.2326 2.6059 0>, rotation=<rot 90 0 0>}
 ]
 spine_04 [
  {shapeType=Capsule, length=8.5796, radius=11.6416, center=<5.5182 0.1467 0>}
  {shapeType=Capsule, length=10.0312, radius=9.2812, center=<-1.2326 1.308 0>, rotation=<rot 90 0 0>}
 ]
 spine_05 [{shapeType=Capsule, length=13.2359, radius=14.5791, center=<13.3628 -0.5845 0>}]
 thigh_l []
 thigh_r []
 upperarm_l []
 upperarm_r []
}
constraints []
